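# Translucent unlit two-sided mesh particle material for a skydiving contrail with a fiber-optic streak look.
# Two panning mask layers tint a two-color gradient, which is desaturation-adjusted and multiplied by a ramp and particle color for emissive.
# Opacity blends those streaks with a packed trail texture, depth-faded except on mobile.
# World-space noise, tuned by dynamic parameters, drives an apparent waving offset that grows away from the object position and is zero on mobile.
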Material /PapayaGameplay/Athena/Effects/Materials/Skydiving/M_Contrail_FiberOptics2
graph {
 properties {
  bUsedWithMeshParticles=true
  bUsedWithNiagaraMeshParticles=true
  blendMode=BLEND_Translucent
  pixelDepthOffsetMode=PDOM_Legacy
  shadingModel=MSM_Unlit
  twoSided=true
  usedShadingModels=Unlit
 }
 nodes [
  n1 = MaterialFunctionCall(Float2=TextureCoordinate()) {materialFunction=&BreakOutFloat2Components}
  n2 = Constant() {r=1}
  n3 = ScalarParameter() {defaultValue=0.65, parameterName=Param_4}
  n4 = LinearInterpolate(
   A=VectorParameter() {parameterName=colorA, defaultValue=<rgba 0.0034 0.3159 1 1>}.RGB
   B=VectorParameter() {parameterName=colorB, defaultValue=<rgba 1 0.3602 0.0038 1>}.RGB
   Alpha=OneMinus(None=SphereMask(A=@n1.G, B=@n2, Radius=@n3) {attenuationRadius=0.35, hardnessPercent=0})
  )
  n5 = MaterialFunctionCall(XY=TextureCoordinate() {uTiling=9, vTiling=0.8}) {materialFunction=&Swizzle}
  n6 = MaterialFunctionCall(XY=TextureCoordinate() {uTiling=5, vTiling=0.65}) {materialFunction=&Swizzle}
  n7 = Add(
   A=TextureSample(UVs=Panner(Coordinate=@n5.YX) {speedX=1.3}) {texture=/Game/Effects/Fort_Effects/Textures/Tile/Tests/T_Tile_Panner_Mask}.R
   B=TextureSample(UVs=Panner(Coordinate=@n6.YX) {speedX=0.8}) {texture=&T_Tile_Panner_Mask}.G
  )
  n8 = Multiply(A=Multiply(A=@n4, B=@n7), B=ScalarParameter() {defaultValue=2.2, parameterName=Param_1})
  n9 = Desaturation(None=Multiply(A=Add(A=@n8, B=@n4)) {constB=0.25}, Fraction=ScalarParameter() {defaultValue=-1.6, parameterName=Param_2})
  n10 = TextureSample() {samplerSource=SSM_Clamp_WorldGroupSettings, texture=/Game/Global/Textures/Default/Masks_and_Gradients/T_Ramp_Virt}
  n11 = ParticleColor()
  MaterialFunctionCall(Emissive=Multiply(A=Multiply(A=@n9, B=@n10.RGB), B=@n11.RGB)) {materialFunction=&MF_FX_EmissiveAdjustments}
  n12 = TextureSample(UVs=MaterialFunctionCall(XY=TextureCoordinate()) {materialFunction=&Swizzle}.YX) {texture=/Game/Effects/Fort_Effects/Textures/Melee/T_Trail_Packed}
  Multiply(A=@n12.R) {constB=5}
  n13 = Multiply(A=@n10.R, B=@n7)
  n14 = MaterialFunctionCall(Float2=TextureCoordinate()) {materialFunction=&BreakOutFloat2Components}
  n15 = Power(Base=@n14.G) {constExponent=2.7}
  n16 = Multiply(A=@n14.G, B=@n14.G)
  n17 = Multiply(A=@n11.A, B=Add(A=Multiply(A=@n13, B=@n15), B=Multiply(A=@n16, B=@n12.R)))
  ShadingPathSwitch(Default=DepthFade(Opacity=@n17), Mobile=@n17) {desc=Migrated from FeatureLevelSwitch}
  n18 = Time()
  n19 = DynamicParameter() {defaultValue=<rgba 1 1 0 1>}
  n20 = ScalarParameter() {defaultValue=0.32, parameterName=Speed}
  n21 = Subtract(
   A=Add(A=Multiply(A=@n18, B=@n19.Param3), B=WorldPosition().XYZ)
   B=Multiply(A=ObjectPositionWS(), B=Add(A=@n20, B=@n19.Param1)) {constB=0.95}
  )
  n22 = ScalarParameter() {defaultValue=680, parameterName=Tilling}
  n23 = TextureSample(UVs=Divide(A=ComponentMask(None=@n21) {b=true}, B=Add(A=@n22, B=@n19.Param2))) {texture=&T_NoisePack_01}
  n24 = ScalarParameter() {defaultValue=500, parameterName=WavingAmount}
  n25 = SphereMask(A=WorldPosition().XYZ, B=ObjectPositionWS(), Radius=ScalarParameter() {defaultValue=5000}, Hardness=Constant())
  ShadingPathSwitch(Default=Multiply(A=Multiply(A=@n23.RGB, B=@n24) {constB=1500}, B=OneMinus(None=@n25)), Mobile=Constant()) {desc=Migrated from FeatureLevelSwitch}
 ]
}
