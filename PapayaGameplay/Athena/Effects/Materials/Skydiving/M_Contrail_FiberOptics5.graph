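# Unlit, two-sided translucent mesh-particle material for a fiber-optic skydiving contrail.
# Two scrolling fibre layers, soft plasma, a time-of-day glow and a distorted fire-mask difference form the emissive, mostly tinted by a three-color V gradient.
# Opacity combines fibres, two-sided axis-aligned fresnel and a V mask eroded by fire noise.
# A static switch selects panned-normal waving or world-space noise displacement growing away from the object, with Dynamic Parameters driving motion and tiling.
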Material /PapayaGameplay/Athena/Effects/Materials/Skydiving/M_Contrail_FiberOptics5
graph {
 properties {
  bUsedWithMeshParticles=true
  bUsedWithNiagaraMeshParticles=true
  blendMode=BLEND_Translucent
  pixelDepthOffsetMode=PDOM_Legacy
  shadingModel=MSM_Unlit
  twoSided=true
  usedShadingModels=Unlit
 }
 nodes [
  n1 = OneMinus(None=TextureCoordinate())
  n2 = Panner() {speedY=-0.2}
  n3 = Multiply(A=ComponentMask(None=TextureSample(UVs=@n2) {samplerType=SAMPLERTYPE_Normal, texture=&T_Weap_Distort_N_02}.RGB) {g=true, r=true}) {constB=0.135}
  n4 = TextureSample(UVs=Panner(Coordinate=Add(A=@n1, B=@n3), Time=Constant() {r=1}) {speedY=0.0045}) {texture=/Game/Effects/Fort_Effects/Textures/Fire/T_Fire_Mask}
  n5 = OneMinus(None=TextureCoordinate())
  n6 = Panner() {speedY=-0.2}
  n7 = Multiply(A=ComponentMask(None=TextureSample(UVs=@n6) {samplerType=SAMPLERTYPE_Normal, texture=&T_Weap_Distort_N_02}.RGB) {g=true, r=true}) {constB=0.135}
  n8 = TextureSample(UVs=Panner(Coordinate=Add(A=@n5, B=@n7))) {texture=&T_Fire_Mask}
  n9 = Multiply(
   A=Multiply(A=Subtract(A=@n4.R, B=@n8.R), B=Constant3Vector() {constant=<rgba 1 0.5419 0 0>})
   B=ScalarParameter() {defaultValue=800, parameterName=Param_5}
  )
  n10 = TextureSample(UVs=TextureCoordinate() {uTiling=0, vTiling=0.75}) {texture=/Game/Effects/Fort_Effects/Textures/Tile/T_Lightning_Plasma_Soft}
  n11 = MaterialFunctionCall(Float2=TextureCoordinate()) {materialFunction=&BreakOutFloat2Components}
  n12 = Constant() {r=1}
  n13 = ScalarParameter() {defaultValue=0.26, parameterName=Param_4}
  n14 = LinearInterpolate(
   A=VectorParameter() {parameterName=colorA, defaultValue=<rgba 0 0.2909 1 1>}.RGB
   B=LinearInterpolate(
    A=VectorParameter() {parameterName=colorB, defaultValue=<rgba 1 0.4292 0.0396 1>}.RGB
    B=VectorParameter() {parameterName=colorC, defaultValue=<rgba 0.3544 0.0377 0.9583 1>}.RGB
    Alpha=Multiply(A=@n11.G, B=@n11.G)
   )
   Alpha=OneMinus(None=SphereMask(A=@n11.G, B=@n12, Radius=@n13) {attenuationRadius=0.35, hardnessPercent=35})
  )
  n15 = MaterialFunctionCall(Float2=TextureCoordinate() {uTiling=1.4, vTiling=0.85}) {materialFunction=&BreakOutFloat2Components}
  n16 = DynamicParameter() {defaultValue=<rgba 1 1 0 420>}
  n17 = Multiply(A=@n15.R, B=@n16.Param2)
  n18 = Constant()
  n19 = Add(A=@n16.Param1, B=Constant() {r=1.85})
  n20 = TextureSample(UVs=Panner(Coordinate=AppendVector(A=@n17, B=@n15.G), Time=Time(), Speed=AppendVector(A=@n18, B=@n19)) {speedY=2.2}) {texture=/Game/Effects/Fort_Effects/Textures/T_fibre_a}
  n21 = MaterialFunctionCall(Float2=TextureCoordinate() {uTiling=0.75, vTiling=0.6}) {materialFunction=&BreakOutFloat2Components}
  n22 = DynamicParameter() {defaultValue=<rgba 1 1 0 420>}
  n23 = Multiply(A=@n21.R, B=@n22.Param2)
  n24 = Constant()
  n25 = Add(A=Constant() {r=1.18}, B=@n22.Param1)
  n26 = TextureSample(UVs=Panner(Coordinate=AppendVector(A=@n23, B=@n21.G), Time=Time(), Speed=AppendVector(A=@n24, B=@n25)) {speedY=2.2}) {texture=&T_fibre_a}
  n27 = Multiply(A=@n14, B=Add(A=@n20.G, B=@n26.G))
  n28 = ScalarParameter() {defaultValue=20, parameterName=Param_3}
  n29 = Constant() {r=7}
  n30 = Constant() {r=3}
  n31 = Multiply(
   A=@n14
   B=MaterialFunctionCall(Color_Morning=@n29, Color_Day=@n29, Color_Evening=@n30, Color_Night=@n30) {materialFunction=&Color_Time_of_Day}.Result
  ) {constB=6}
  n32 = Constant() {r=-1.4}
  n33 = Constant() {r=-0.75}
  n34 = MaterialFunctionCall(Color_Morning=@n32, Color_Day=@n32, Color_Evening=@n33, Color_Night=@n33) {materialFunction=&Color_Time_of_Day}
  n35 = Add(
   A=Multiply(A=Multiply(A=@n10.B, B=@n14)) {constB=500}
   B=Add(A=Multiply(A=@n27, B=@n28) {constB=46}, B=Desaturation(None=@n31, Fraction=@n34.Result))
  )
  n36 = ParticleColor()
  MaterialFunctionCall(Emissive=Multiply(A=Add(A=@n9, B=@n35), B=@n36.RGB)) {materialFunction=&MF_FX_EmissiveAdjustments}
  n37 = MaterialFunctionCall!(Axis, Exponent, Normal=Multiply(A=VertexNormalWS()) {constB=-1}) {materialFunction=/Game/Effects/Fort_Effects/Materials/_Functions/MF_AxisAlignedFrensnel}
  n38 = MaterialFunctionCall() {materialFunction=&MF_AxisAlignedFrensnel}
  n39 = TwoSidedSign()
  n40 = Add(A=Add(A=@n20.R, B=@n20.R), B=Add(A=@n26.R, B=@n26.R))
  n41 = SphereMask(
   A=MaterialFunctionCall(Float2=TextureCoordinate()) {materialFunction=&BreakOutFloat2Components}.G
   B=Constant() {r=0.5}
  ) {attenuationRadius=0.46, hardnessPercent=0}
  n42 = Subtract(A=Multiply(A=@n41, B=@n41), B=@n8.R)
  Multiply(A=@n36.A, B=Multiply(A=LinearInterpolate(A=@n37.Fresnel, B=@n38.Fresnel, Alpha=@n39), B=Multiply(A=@n40, B=@n42)))
  n43 = Panner(Coordinate=TextureCoordinate() {uTiling=0.32, vTiling=0.9}) {speedY=0.88}
  n44 = Time()
  n45 = DynamicParameter() {defaultValue=<rgba 1 1 0 420>}
  n46 = ScalarParameter() {parameterName=Speed}
  n47 = Subtract(
   A=Add(A=Multiply(A=@n44, B=@n45.Param3), B=WorldPosition().XYZ)
   B=Multiply(A=ObjectPositionWS(), B=Add(A=@n46, B=@n45.Param1)) {constB=0.95}
  )
  n48 = ScalarParameter() {defaultValue=1100, parameterName=Tilling}
  n49 = TextureSample(UVs=Divide(A=ComponentMask(None=@n47) {b=true}, B=Add(A=@n48, B=@n45.Param2))) {texture=&T_NoisePack_01}
  n50 = DynamicParameter() {defaultValue=<rgba 1 1 0 420>}
  n51 = SphereMask(A=WorldPosition().XYZ, B=ObjectPositionWS(), Radius=ScalarParameter() {defaultValue=2000}, Hardness=Constant())
  StaticSwitchParameter(
   True=Multiply(
    A=TextureSample(UVs=@n43) {samplerType=SAMPLERTYPE_Normal, texture=&T_Weap_Distort_N_02}.RGB
    B=ScalarParameter() {defaultValue=420, parameterName=WavingAmount}
   ) {constB=1500}
   False=Multiply(A=Multiply(A=@n49.RGB, B=@n50.Param4) {constB=1500}, B=OneMinus(None=@n51))
  )
 ]
}
